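; Server-authoritative manager that decides whether the VividRazor wolf boss spawns at the train or at a bank vault.
; Once playlist data is ready, it checks a spawn-elsewhere override and can wait for the BusLocked or StormForming game phase event before making a one-time pick.
; Scalable floats choose equal odds or a train percentage, with train fallback when no vaults exist.
; A picked vault's event manager is flagged; train selection is exposed through a query.
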(event EventBeginPlay
  ; Begin Play - Wait for Bus Locked or Storm Forming to pick where the Boss should be spawning.
  (Networking|SwitchHasAuthority
    (:Authority
      (Variables|InternalData|SetSpawnAtTrain false)
      (Class|FortAsyncActionCurrentPlaylistReady|PlaylistDataReadyAsync
        (:OnReady
          (if (ScalableFloat|GetValueasBool (Variables|Default|GetShouldBossSpawnElsewhere))
            (Fortnite|FortShippingLog self "We have chosen to make the Boss spawn Elsewhere, therefore he should not be spawning at the train nor a bank vault.")
            (elif (Variables|Default|GetWaitForBusLockedToChooseLocation)
              (bind _returnvalue (GameplayEventRouter|GetEventRouter "/Script/FortniteGame.FortGameState" self))
              (bind (_asynctaskproxy _payload) (GameplayEventRouter|StartListeningToStatefulEvent _returnvalue "/Script/FortniteGame.FortGamePhaseStepUpdatedEvent")
                (:then
                  (Variables|Default|SetGamePhaseStepUpdatedAsyncTask _asynctaskproxy))
                (:OnEventReceived
                  _)
                (:OnSavedEventStateExists
                  _))
              (if (or (Utilities|Enum|Equal(Enum) (Utilities|Struct|BreakFortGamePhaseStepUpdatedEvent _payload) "BusLocked") (Utilities|Enum|Equal(Enum) (Utilities|Struct|BreakFortGamePhaseStepUpdatedEvent _payload) "StormForming"))
                (GameplayEventRouter|StopListeningtoStatefulEvent (Variables|Default|GetGamePhaseStepUpdatedAsyncTask)))))
          (Utilities|FlowControl|DoOnce false
            (:Completed
              (CallFunction|FindBossSpawnLocation))))))))

(event Custom|FindBossSpawnLocation
  ; Figure out where the Boss will spawn!
  (if (ScalableFloat|GetValueasBool (Variables|Default|GetUseEqualChancesToSpawnOnTrainOrVaults))
    (if (not (== (Math|Random|RandomIntegerinRange 1 (+ (Utilities|Array|Length (Variables|Default|GetVaultManagers)) 1)) 1))
      (bind _element (ArrayMacros|GetRandomArrayElement (Variables|Default|GetVaultManagers))
        (:%20%20
          (Variables|InternalData|SetPickedVault _element))
        (:Empty%20Array
          (Variables|InternalData|SetSpawnAtTrain true :!label _label)))
      (CallFunction|AlertSelectedLocation))
    (goto _label)
    (elif (<= (ScalableFloat|GetValueatLevel (Variables|Default|GetTrainPercentageChangeToBePicked)) (Math|Random|RandomFloatinRange 0.0 1.0))
      _
      (else
        (bind _element_2 (ArrayMacros|GetRandomArrayElement (Variables|Default|GetVaultManagers))
          (:%20%20
            (Variables|InternalData|SetPickedVault _element_2))
          (:Empty%20Array
            _)))))
  (Variables|InternalData|SetSpawnAtTrain true))

(event Custom|AlertSelectedLocation
  ; Alert, if needed, where the Boss will be spawning!
  (if (Variables|InternalData|GetSpawnAtTrain)
    (Development|PrintString "WOLF BOSS WILL BE SPAWNING AT THE TRAIN!!!" true false "(R=0.822917,G=0.304294,B=0.184871,A=1.000000)" 6e+01)
    (else
      (Class|BPVividRazorEventManager|SetBossShouldSpawnHere (Variables|InternalData|GetPickedVault) :!class /VividRazor_Minions/Blueprints/BP_VividRazor_EventManager.BP_VividRazor_EventManager_C)
      (Development|PrintString "WOLF BOSS WILL BE SPAWNING AT A VAULT!!!" true false "(R=0.822917,G=0.304294,B=0.184871,A=1.000000)" 6e+01))))

(fn IsBossSupposedToSpawnAtTrain
  (return (Variables|InternalData|GetSpawnAtTrain)))

(fn ConstructionScript)
