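; Server-side manager for a VividRazor vault encounter: patrolling guards, staged reinforcements, an optional boss and a post-opening cleanup stage.
; Playspace event router payloads, filtered by vault location tag, start, advance, pause or wrap up the Living World encounter.
; Timers, weakpoint breaks and kill quotas advance stages, tuned by team-size scalable floats.
; Area volumes drive guard hostility, spawns receive deferred AI setup, and an AI density reservoir reserves spawn budget.
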(event EventBeginPlay
  (Networking|SwitchHasAuthority
    (:Authority
      (CallFunction|BindVaultAreaVolumeEvents)
      (CallFunction|BindEncounterAreaVolumeEvents)
      (bind (_gamestate _playlist) (Class|FortAsyncActionCurrentPlaylistReady|PlaylistDataReadyAsync)
        (:OnReady
          (Variables|VaultEncounterManager|SetTeamSize (Variables|/Script/FortniteGame.FortPlaylist|GetMaxTeamSize _playlist))
          (CallFunction|ValidateTeamSize)
          (Utilities|FlowControl|DelayUntilNextTick)
          (CallFunction|ReserveAIDensityReservoir true)
          (CallFunction|SpawnPatrollingGuards))))))

(event EventOnWorldReady
  (|CallParentFunction)
  (Networking|SwitchHasAuthority
    (:Authority
      (bind (_asynctaskproxy _eventrouter) (Class|AsyncActionWaitforPlayspaceEventRouter|WaitforPlayspaceEventRouter 0 "/VividRazorVaultGameplay/Gameplay/Environmental/Manager/BP_VividRazorVaultPlayspace.BP_VividRazorVaultPlayspace_C" true true)
        (:OnEventRouterAvailable
          (Variables|Default|SetVaultEventRouter _eventrouter)
          (CallFunction|ListenforAlertStarted)
          (CallFunction|ListenforWeakpointDestroyed)
          (CallFunction|ListenforVaultPaused)
          (CallFunction|ListenforVaultOpened)
          (Utilities|FlowControl|Delay 1.0) ; The vault manager is assigned once\nthe router becomes available.
          (bind _asbp_vivid_razor_vault_playspace (Utilities|Casting|CastToBP_VividRazorVaultPlayspace (Components|GetOwner (Variables|Default|GetVaultEventRouter))))
          (bind _output_get (Variables|VaultEncounterManager|SetVaultManager (Variables|/VividRazorVaultGameplay/Gameplay/Environmental/Manager/BP_VividRazorVaultPlayspace.BP_VividRazorVaultPlayspace_C|GetVaultManager _asbp_vivid_razor_vault_playspace)))
          (if (and (Utilities|IsValid _output_get :!class /Script/Engine.KismetSystemLibrary) (or (Variables|VaultEncounterManager|GetBossShouldSpawnHere) (and (Editor|InEditororPIE) (Variables|VaultEncounterManager|GetDebugSetBossSpawnHere))))
            (Class|BGAVividRazorVaultManager|DisplayBossonMinimap (Variables|VaultEncounterManager|GetVaultManager) true)))))))

(event Custom|ListenForAlertStarted
  (Development|PrintString "[Encounter Manager] Starting to listen for Alert event from the vault." (Variables|VaultEncounterManager|GetVerboseLogging) (Variables|VaultEncounterManager|GetVerboseLogging) (Variables|VaultEncounterManager|GetDebugOutputColor) 6.0)
  (bind (_asynctaskproxy _payload) (GameplayEventRouter|StartListeningToEvent (Variables|Default|GetVaultEventRouter) "/VividRazorCoreGameplay/Gameplay/Structs/Struct_VividRazorCore_Vault_Alert_Payload.Struct_VividRazorCore_Vault_Alert_Payload")
    (:OnEventReceived
      (Development|PrintString "[Encounter Manager] Received Alert event from the vault!" (Variables|VaultEncounterManager|GetVerboseLogging) (Variables|VaultEncounterManager|GetVerboseLogging) (Variables|VaultEncounterManager|GetDebugOutputColor) 6.0)
      (if (== (Utilities|Struct|BreakStructVividRazorCoreVaultAlertPayload _payload) (Variables|VaultEncounterManager|GetVaultLocationTag))
        (CallFunction|OnAlertInitiated)
        (Variables|Default|SetSoundDispatcher (Actor|GetComponentbyClass (Utilities|ResolveSoftReference (.VaultManagerActor_28_2F604354450CF5624125F284BA0BA87B (Utilities|Struct|BreakStructVividRazorCoreVaultAlertPayload _payload))) "/VividRazorVaultGameplay/Assets/Audio/Blueprints/SoundEventDispatcher_VividRazerVault.SoundEventDispatcher_VividRazerVault_C"))))))

(event Custom|ListenForWeakpointDestroyed
  (bind (_asynctaskproxy _payload) (GameplayEventRouter|StartListeningToEvent (Variables|Default|GetVaultEventRouter) "/VividRazorCoreGameplay/Gameplay/Structs/Struct_VividRazorCore_Vault_Weakpoint_Destroyed_Payload.Struct_VividRazorCore_Vault_Weakpoint_Destroyed_Payload")
    (:OnEventReceived
      (if (== (Utilities|Struct|BreakStructVividRazorCoreVaultWeakpointDestroyedPayload _payload) (Variables|VaultEncounterManager|GetVaultLocationTag))
        (CallFunction|UpdateVaultProgress)
        (if (and (> (ScalableFloat|GetValueatLevel (Variables|VaultEncounterManager|GetUseWeakpointStageTriggers) (Math|Conversions|ToFloat(Integer) (Variables|VaultEncounterManager|GetTeamSize))) 0) (< (Variables|VaultEncounterManager|GetCurrentVaultProgress) 1.0))
          (bind _returnvalue (Utilities|Time|SetTimerbyFunctionName 0 "AdvanceEncounterStage" (ScalableFloat|GetValueatLevel (Variables|VaultEncounterManager|GetDelayBetweenStages) (Math|Conversions|ToFloat(Integer) (Variables|VaultEncounterManager|GetTeamSize)))))
          (Variables|VaultEncounterManager|SetAdvanceStageTimer _returnvalue))))))

(event Custom|ListenForVaultPaused
  (bind _advancestagetimer (Variables|VaultEncounterManager|GetAdvanceStageTimer))
  (bind _bossspawntimerhandle (Variables|VaultEncounterManager|GetBossSpawnTimerHandle))
  (bind (_asynctaskproxy _payload) (GameplayEventRouter|StartListeningToEvent (Variables|Default|GetVaultEventRouter) "/VividRazorCoreGameplay/Gameplay/Structs/Struct_VividRazorCore_Vault_Paused_Payload.Struct_VividRazorCore_Vault_Paused_Payload")
    (:OnEventReceived
      (if (== (Utilities|Struct|BreakStructVividRazorCoreVaultPausedPayload _payload) (Variables|VaultEncounterManager|GetVaultLocationTag))
        (CallFunction|PauseEncounter)
        (CallFunction|SubmitAnalytics)
        (if (Utilities|Time|IsTimerActivebyHandle _advancestagetimer)
          (Utilities|Time|PauseTimerbyHandle _advancestagetimer)
          (else
            _))
        (if (Utilities|Time|IsTimerActivebyHandle _bossspawntimerhandle)
          (Utilities|Time|PauseTimerbyHandle _bossspawntimerhandle))))))

(event Custom|ListenForVaultOpened
  (bind (_asynctaskproxy _payload) (GameplayEventRouter|StartListeningToStatefulEvent (Variables|Default|GetVaultEventRouter) "/VividRazorCoreGameplay/Gameplay/Structs/Struct_VividRazorCore_Vault_Opened_Payload.Struct_VividRazorCore_Vault_Opened_Payload")
    (:OnEventReceived
      (if (== (Utilities|Struct|BreakStructVividRazorCoreVaultOpenedPayload _payload) (Variables|VaultEncounterManager|GetVaultLocationTag))
        (CallFunction|OnVaultCracked)))))

(event Custom|TriggerCleanupStage
  (bind _encounter (Variables|VaultEncounterManager|GetEncounter))
  (if (> (Utilities|Array|Length (Variables|VaultEncounterManager|GetActivePlayers)) 0)
    (CallFunction|ToggleReinforcementPointProviders)
    (Class|FortAthenaLivingWorldEncounterInstance|StartEncounterStage _encounter (- (Class|FortAthenaLivingWorldEncounterInstance|GetNumberOfStages _encounter) 1))
    (CallFunction|ResetStageData)
    (Class|FortAthenaLivingWorldEncounterInstance|ResumeEncounter _encounter)
    (else
      (Default|EndEncounter))))

(event Custom|OnAlertInitiated
  (bind _teamsize (Variables|VaultEncounterManager|GetTeamSize))
  (bind _returnvalue (Math|Conversions|ToFloat(Integer) _teamsize))
  (bind _initialreinforcementsdelay (Variables|VaultEncounterManager|GetInitialReinforcementsDelay))
  (bind _returnvalue_2 (ScalableFloat|GetValueatLevel _initialreinforcementsdelay _returnvalue))
  (bind _returnvalue_3 (Math|Float|Clamp(Float) _returnvalue_2 0.1 9999.0))
  (bind _advancestagetimer (Variables|VaultEncounterManager|GetAdvanceStageTimer))
  (bind _bossspawntimerhandle (Variables|VaultEncounterManager|GetBossSpawnTimerHandle))
  (if (Variables|VaultEncounterManager|GetEncounterStarted)
    (if (Utilities|Time|IsTimerPausedbyHandle _advancestagetimer)
      (Utilities|Time|ClearandInvalidateTimerbyHandle _advancestagetimer)
      (else
        _))
    (bind _returnvalue_4 (Utilities|Time|SetTimerbyFunctionName 0 "AdvanceEncounterStage" (ScalableFloat|GetValueatLevel (Variables|VaultEncounterManager|GetEncounterRestartReinforcementsDelay) (Math|Conversions|ToFloat(Integer) (Variables|VaultEncounterManager|GetTeamSize)))))
    (Variables|VaultEncounterManager|SetAdvanceStageTimer _returnvalue_4)
    (if (Utilities|Time|IsTimerPausedbyHandle _bossspawntimerhandle)
      (Utilities|Time|UnpauseTimerbyHandle _bossspawntimerhandle))
    (else
      (Variables|VaultEncounterManager|SetEncounterStarted true)
      (CallFunction|AlertEnemies)
      (Development|PrintString (Utilities|String|Append "[Encounter Manager] Starting the encounter in " (Utilities|String|ToString(Float) _returnvalue_3) " seconds!") (Variables|VaultEncounterManager|GetVerboseLogging) (Variables|VaultEncounterManager|GetVerboseLogging) (Variables|VaultEncounterManager|GetDebugOutputColor) 6.0)
      (Utilities|Time|SetTimerbyEvent (EventDispatchers|CreateEvent :!event Custom|OnStartEncounter) _returnvalue_3)
      (if (and (Variables|VaultEncounterManager|GetBossShouldSpawnHere) (not (Variables|VaultEncounterManager|GetBossSpawned)))
        (bind _returnvalue_5 (Utilities|Time|SetTimerbyEvent (EventDispatchers|CreateEvent :!event Custom|OnBossTimerDone) (Math|Random|RandomFloatinRange (ScalableFloat|GetValueatLevel (Variables|VaultEncounterManager|GetBossSpawnTimerMin) (Math|Conversions|ToFloat(Integer) (Variables|VaultEncounterManager|GetTeamSize))) (ScalableFloat|GetValueatLevel (Variables|VaultEncounterManager|GetBossSpawnTimerMax) (Math|Conversions|ToFloat(Integer) (Variables|VaultEncounterManager|GetTeamSize))))))
        (Variables|VaultEncounterManager|SetBossSpawnTimerHandle _returnvalue_5)))))

(event Custom|OnStartEncounter
  (Development|PrintString "[Encounter Manager] Starting the encounter now!" (Variables|VaultEncounterManager|GetVerboseLogging) (Variables|VaultEncounterManager|GetVerboseLogging) (Variables|VaultEncounterManager|GetDebugOutputColor) 6.0)
  (CallFunction|StartEncounter))

(event Custom|OnVaultCracked
  (bind _teamsize (Variables|VaultEncounterManager|GetTeamSize))
  (bind _returnvalue (Math|Conversions|ToFloat(Integer) _teamsize))
  (bind _cleanupstagedelay (Variables|VaultEncounterManager|GetCleanupStageDelay))
  (bind _returnvalue_2 (ScalableFloat|GetValueatLevel _cleanupstagedelay _returnvalue))
  (Variables|VaultEncounterManager|SetVaultCracked true)
  (CallFunction|ReserveAIDensityReservoir false)
  (Utilities|Time|ClearandInvalidateTimerbyHandle (Variables|VaultEncounterManager|GetAdvanceStageTimer))
  (CallFunction|PauseEncounter)
  (if (> _returnvalue_2 0)
    (Utilities|Time|SetTimerbyEvent (EventDispatchers|CreateEvent :!event Custom|TriggerCleanupStage) _returnvalue_2)
    (else
      (Default|EndEncounter))))

(event Custom|OnActorEnteredVaultArea
  (params (OverlappedActor :type Actor%20Object%20Reference) (OtherActor :type Actor%20Object%20Reference))
  (bind _asfort_pawn (Utilities|Casting|CastToFortPawn OtherActor))
  (if (Pawn|IsPlayerControlled _asfort_pawn)
    (CallFunction|MakeEnemiesHostileTowardsPlayer _asfort_pawn)))

(event Custom|OnActorLeftVaultArea
  (params (OverlappedActor :type Actor%20Object%20Reference) (OtherActor :type Actor%20Object%20Reference))
  (bind _asfort_pawn (Utilities|Casting|CastToFortPawn OtherActor))
  (if (Pawn|IsPlayerControlled _asfort_pawn)
    (CallFunction|MakeEnemiesNeutralTowardsPlayer _asfort_pawn)))

(event Custom|OnActorEnteredEncounterArea
  (params (OverlappedActor :type Actor%20Object%20Reference) (OtherActor :type Actor%20Object%20Reference))
  (bind _asfort_pawn (Utilities|Casting|CastToFortPawn OtherActor))
  (if (Pawn|IsPlayerControlled _asfort_pawn)
    (Development|PrintString "[Encounter Manager] Player entered encounter area!" (Variables|VaultEncounterManager|GetVerboseLogging) (Variables|VaultEncounterManager|GetVerboseLogging) (Variables|VaultEncounterManager|GetDebugOutputColor))
    (CallFunction|HandlePlayerEnteredEncounterArea _asfort_pawn)
    (elif (CallFunction|IsPawnBoss _asfort_pawn)
      (CallFunction|HandleBossSpawned _asfort_pawn))))

(event Custom|OnActorLeftEncounterArea
  (params (OverlappedActor :type Actor%20Object%20Reference) (OtherActor :type Actor%20Object%20Reference))
  (bind _asfort_pawn (Utilities|Casting|CastToFortPawn OtherActor))
  (if (Pawn|IsPlayerControlled _asfort_pawn)
    (CallFunction|HandlePlayerLeftEncounterArea _asfort_pawn)))

(event Custom|OnEnemyAlertLevelChanged
  (params (BotController :type Fort%20Athena%20AIBot%20Controller%20Object%20Reference) (OldAlertLevel :type EAlertLevel%20Enum :default "Unaware") (NewAlertLevel :type EAlertLevel%20Enum :default "Unaware"))
  (if (Utilities|Enum|Equal(Enum) NewAlertLevel "Threatened")
    (Variables|VaultEncounterManager|SetCombatStarted true)))

(event EventEndPlay
  (params (EndPlayReason :type EEndPlayReason%20Enum :default "Destroyed"))
  (CallFunction|ReserveAIDensityReservoir false)
  (CallFunction|SubmitAnalytics))

(event Custom|SetBossShouldSpawnHere
  ; Called from the Boss Manager to know that the Boss should be spawning from This Vault!
  (Variables|VaultEncounterManager|SetBossShouldSpawnHere true)
  (bind _vaultmanager (Variables|VaultEncounterManager|GetVaultManager)
    (:then
      (Class|BGAVividRazorVaultManager|DisplayBossonMinimap _vaultmanager true))
    (:else)))

(event Custom|OnBossTimerDone
  (CallFunction|SpawnBoss))

(event Custom|OnGuardReceivedDamage
  (params (DamagedActor :type Actor%20Object%20Reference) (Damage :type Float%20%28single-precision%29) (InstigatedBy :type Controller%20Object%20Reference) (DamageCauser :type Actor%20Object%20Reference) (HitLocation :type Vector) (FHitComponent :type Primitive%20Component%20Object%20Reference) (BoneName :type Name) (Momentum :type Vector))
  (if (Pawn|IsPlayerController InstigatedBy)
    (Variables|Analytics|SetDamageToGuards (+ (Variables|Analytics|GetDamageToGuards) Damage))))

(event Custom|OnGuardDealtDamage
  (params (DamagedActor :type Actor%20Object%20Reference) (Damage :type Float%20%28single-precision%29) (InstigatedBy :type Controller%20Object%20Reference) (DamageCauser :type Actor%20Object%20Reference) (HitLocation :type Vector) (FHitComponent :type Primitive%20Component%20Object%20Reference) (BoneName :type Name) (Momentum :type Vector))
  (bind _asfort_pawn (Utilities|Casting|CastToFortPawn DamagedActor))
  (if (Pawn|IsPlayerControlled _asfort_pawn)
    (Variables|Analytics|SetDamageByGuards (+ (Variables|Analytics|GetDamageByGuards) Damage))))

(fn AdvanceEncounterStage
  (bind _teamsize (Variables|VaultEncounterManager|GetTeamSize))
  (bind _returnvalue (Math|Conversions|ToFloat(Integer) _teamsize))
  (bind _delaybetweenstages (Variables|VaultEncounterManager|GetDelayBetweenStages))
  (bind _returnvalue_2 (ScalableFloat|GetValueatLevel _delaybetweenstages _returnvalue))
  (bind _returnvalue_3 (Math|Float|Clamp(Float) _returnvalue_2 0.1 9999.0))
  (bind _encounter (Variables|VaultEncounterManager|GetEncounter))
  (bind _encounter_2 (Variables|VaultEncounterManager|GetEncounter))
  (Utilities|IsValid _encounter_2
    (:Is%20Valid
      (Development|PrintString (Utilities|String|Append "[Encounter Manager] Starting stage " (Utilities|String|ToString(Integer) (+ (Variables|/Script/LagerRuntime.FortAthenaLivingWorldEncounterInstance|GetCurrentStageIndex _encounter_2) 2)) " of the encounter!") (Variables|VaultEncounterManager|GetVerboseLogging) true (Variables|VaultEncounterManager|GetDebugOutputColor) 4.0)
      (CallFunction|ToggleReinforcementPointProviders)
      (Class|FortAthenaLivingWorldEncounterInstance|AdvanceEncounterStage _encounter)
      (CallFunction|ResetStageData)
      (Class|FortAthenaLivingWorldEncounterInstance|ResumeEncounter _encounter)
      (bind _sounddispatcher (Variables|Default|GetSoundDispatcher)
        (:then
          (Class|SoundEventDispatcherVividRazerVault|AdvanceEncounterState _sounddispatcher))
        (:else
          _))
      (if (and (> (ScalableFloat|GetValueatLevel (Variables|VaultEncounterManager|GetUseTimedStageTriggers) (Math|Conversions|ToFloat(Integer) (Variables|VaultEncounterManager|GetTeamSize))) 0) (< (+ (Variables|/Script/LagerRuntime.FortAthenaLivingWorldEncounterInstance|GetCurrentStageIndex (Variables|VaultEncounterManager|GetEncounter)) 1) (Class|FortAthenaLivingWorldEncounterInstance|GetNumberOfStages (Variables|VaultEncounterManager|GetEncounter))))
        (bind _returnvalue_4 (Utilities|Time|SetTimerbyFunctionName 0 "AdvanceEncounterStage" _returnvalue_3))
        (Variables|VaultEncounterManager|SetAdvanceStageTimer _returnvalue_4)
        (Development|PrintString (Utilities|String|Append "[Encounter Manager] Queueing the next stage in " (Utilities|String|ToString(Float) _returnvalue_3) " seconds!") (Variables|VaultEncounterManager|GetVerboseLogging) true (Variables|VaultEncounterManager|GetDebugOutputColor) 4.0)))))

(fn AlertEnemies
  (for _array_element (Variables|VaultEncounterManager|GetActiveEnemies)
    (bind _asfort_athena_aibot_controller (Utilities|Casting|CastToFortAthenaAIBotController (Pawn|GetController _array_element)))
    (Class|FortAthenaAIBotController|QueueStim _asfort_athena_aibot_controller (select (Variables|VaultEncounterManager|GetVaultCracked) (Variables|VaultEncounterManager|GetVaultInteriorTargetPoint) (Variables|VaultEncounterManager|GetVaultTargetPoint)) (Math|Vector|SelectVector (Transformation|GetActorLocation (Variables|VaultEncounterManager|GetVaultTargetPoint)) (Transformation|GetActorLocation (Variables|VaultEncounterManager|GetVaultInteriorTargetPoint)) (not (Variables|VaultEncounterManager|GetVaultCracked))) (- (Math|Vector|SelectVector (Transformation|GetActorLocation (Variables|VaultEncounterManager|GetVaultTargetPoint)) (Transformation|GetActorLocation (Variables|VaultEncounterManager|GetVaultInteriorTargetPoint)) (not (Variables|VaultEncounterManager|GetVaultCracked))) (Transformation|GetActorLocation _array_element)) "Enemy" 1e+02 "None")
    (for _array_element_2 (Variables|VaultEncounterManager|GetActivePlayers)
      (Affiliation|ChangeAttitudeBetweenActors _array_element _array_element_2 "Hostile"))))

(fn BindEncounterAreaVolumeEvents
  (bind _encounterareavolume (Variables|VaultEncounterManager|GetEncounterAreaVolume))
  (Utilities|IsValid _encounterareavolume
    (:Is%20Valid
      (Collision|BindEventtoOnActorBeginOverlap _encounterareavolume (EventDispatchers|CreateEvent :!event OnActorEnteredEncounterArea))
      (Collision|BindEventtoOnActorEndOverlap _encounterareavolume (EventDispatchers|CreateEvent :!event OnActorLeftEncounterArea)))))

(fn BindVaultAreaVolumeEvents
  (bind _vaultareavolume (Variables|VaultEncounterManager|GetVaultAreaVolume))
  (Utilities|IsValid _vaultareavolume
    (:Is%20Valid
      (Collision|BindEventtoOnActorBeginOverlap _vaultareavolume (EventDispatchers|CreateEvent :!event OnActorEnteredVaultArea))
      (Collision|BindEventtoOnActorEndOverlap _vaultareavolume (EventDispatchers|CreateEvent :!event OnActorLeftVaultArea)))))

(fn EndEncounter
  (bind _encounter (Variables|VaultEncounterManager|GetEncounter))
  (Utilities|IsValid _encounter
    (:Is%20Valid
      (Class|BPFLLagerCommunFunctions|EndActivityEncounter (Game|GetGameState) _encounter)
      (EventDispatchers|UnbindallEventsfromOnActorSpawned _encounter)
      (EventDispatchers|UnbindallEventsfromOnActorDied _encounter))
    (:Is%20Not%20Valid
      _))
  (Default|CallOnEventEnded)
  (bind _sounddispatcher (Variables|Default|GetSoundDispatcher)
    (:then
      (Class|SoundEventDispatcherVividRazerVault|EncounterEnd _sounddispatcher))
    (:else
      _))
  (Development|PrintString "[Encounter Manager] Last stage of the encounter finished. Encounter completed!" (Variables|VaultEncounterManager|GetVerboseLogging) true (Variables|VaultEncounterManager|GetDebugOutputColor) 4.0)
  (bind _vaultmanager (Variables|VaultEncounterManager|GetVaultManager)
    (:then
      (bind _vaultactor (Variables|/VividRazorVaultGameplay/Gameplay/Environmental/Manager/BGA_VividRazorVault_Manager.BGA_VividRazorVault_Manager_C|GetVaultActor _vaultmanager)
        (:then
          (Class|BGAVividRazorVaultDoorParent|HandleEncounterCompleted _vaultactor)
          (CallFunction|SubmitAnalytics))
        (:else)))
    (:else)))

(fn HandleActorDied
  (params (EncounterInstance :type Fort%20Athena%20Living%20World%20Encounter%20Instance%20Object%20Reference) (DiedActor :type Actor%20Object%20Reference))
  (bind _asfort_pawn (Utilities|Casting|CastToFortPawn DiedActor))
  (Utilities|Array|RemoveItem (Variables|VaultEncounterManager|GetActiveEnemies) _asfort_pawn)
  (Math|Integer|IncrementInt (Variables|Analytics|GetActorsKilledTotal)
    (:%20%20
      (Math|Integer|IncrementInt (Variables|VaultEncounterManager|GetActorsKilledInCurrentStage)
        (:%20%20
          (Development|PrintString (Utilities|String|Append "[Encounter Manager]  Guard killed! Guards killed in this stage: " (Utilities|String|ToString(Integer) (Variables|VaultEncounterManager|GetActorsKilledInCurrentStage)) ".   Guards still alive: " (Utilities|String|ToString(Integer) (Utilities|Array|Length (Variables|VaultEncounterManager|GetActiveEnemies)))) (Variables|VaultEncounterManager|GetVerboseLogging) true (Variables|VaultEncounterManager|GetDebugOutputColor) 4.0)
          (if (> (ScalableFloat|GetValueatLevel (Variables|VaultEncounterManager|GetUseKillPercentageStageTriggers) (Math|Conversions|ToFloat(Integer) (Variables|VaultEncounterManager|GetTeamSize))) 0)
            (bind _killquotamet (CallFunction|HasKillQuotaBeenMet))
            (if _killquotamet
              (Default|HandleKillQuotaMet))))))))

(fn HandleActorDiedProxyFunction
  (params (DamagedActor :type Actor%20Object%20Reference) (Damage :type Float%20%28single-precision%29) (InstigatedBy :type Controller%20Object%20Reference) (DamageCauser :type Actor%20Object%20Reference) (HitLocation :type Vector) (FHitComponent :type Primitive%20Component%20Object%20Reference) (BoneName :type Name) (Momentum :type Vector))
  (CallFunction|HandleActorDied 0 DamagedActor))

(fn HandleActorSpawned
  (params (EncounterInstance :type Fort%20Athena%20Living%20World%20Encounter%20Instance%20Object%20Reference) (SpawnedActor :type Actor%20Object%20Reference) (EventData :type Fort%20Athena%20Living%20World%20Event%20Data%20Object%20Reference))
  (bind _encounter (Variables|VaultEncounterManager|GetEncounter))
  (bind _asfort_pawn (Utilities|Casting|CastToFortPawn SpawnedActor))
  (bind _asfort_athena_aibot_controller (Utilities|Casting|CastToFortAthenaAIBotController (Pawn|GetController _asfort_pawn)))
  (Utilities|Map|Add (Variables|VaultEncounterManager|GetUninitializedEnemies) _asfort_pawn (Utilities|Time|GetTimeSeconds))
  (bind _returnvalue (Utilities|Array|AddUnique (Variables|VaultEncounterManager|GetActiveEnemies) _asfort_pawn))
  (if (>= _returnvalue 0)
    (EventDispatchers|BindEventtoOnDamaged _asfort_pawn (EventDispatchers|CreateEvent :!event OnGuardReceivedDamage))
    (EventDispatchers|BindEventtoOnDidDamage _asfort_pawn (EventDispatchers|CreateEvent :!event OnGuardDealtDamage))
    (AlertLevel|BindEventtoOnAlertLevelChangedEvent _asfort_athena_aibot_controller (EventDispatchers|CreateEvent :!event OnEnemyAlertLevelChanged))
    (Math|Integer|IncrementInt (Variables|VaultEncounterManager|GetActorsSpawnedInCurrentStage)
      (:%20%20
        (Class|FortAthenaAIBotController|SetRadialLeashLocation _asfort_athena_aibot_controller (Transformation|GetActorLocation (Variables|VaultEncounterManager|GetLeashCenter)))
        (Class|FortAthenaAIBotController|SetRadialLeashInnerRadius _asfort_athena_aibot_controller (Variables|VaultEncounterManager|GetLeashInnerRadius))
        (Class|FortAthenaAIBotController|SetRadialLeashOuterRadius _asfort_athena_aibot_controller (Variables|VaultEncounterManager|GetLeashOuterRadius))
        (Development|PrintString (Utilities|String|Append "[Encounter Manager] Guard spawned. Currently active guards: " (Utilities|String|ToString(Integer) (Utilities|Array|Length (Variables|VaultEncounterManager|GetActiveEnemies)))) (Variables|VaultEncounterManager|GetVerboseLogging) true (Variables|VaultEncounterManager|GetDebugOutputColor) 4.0)
        (Utilities|FlowControl|Sequence
          (:then_0
            (if (not (Utilities|Time|IsTimerActivebyHandle (Variables|VaultEncounterManager|GetEnemyInitializationTimer)))
              (bind _returnvalue_2 (Utilities|Time|SetTimerbyFunctionName 0 "InitializeReadyEnemies" 0.5))
              (Variables|VaultEncounterManager|SetEnemyInitializationTimer _returnvalue_2)))
          (:then_1
            (if (Utilities|IsValid _encounter :!class /Script/Engine.KismetSystemLibrary)
              (if (>= (/ (Math|Conversions|ToFloat(Integer) (Variables|VaultEncounterManager|GetActorsSpawnedInCurrentStage)) (/ (Variables|/Script/LagerRuntime.FortAthenaLivingWorldEncounterInstance|GetCurrentStageActorMaxCount _encounter) 2)) (ScalableFloat|GetValueatLevel (Variables|VaultEncounterManager|GetAcceptableSpawnPercentage) (Math|Conversions|ToFloat(Integer) (Variables|VaultEncounterManager|GetTeamSize))))
                (Utilities|Time|SetTimerbyFunctionName 0 "PauseSpawning" 2.0)))))))))

(fn HandleBossDied
  (params (DamagedActor :type Actor%20Object%20Reference) (Damage :type Float%20%28single-precision%29) (InstigatedBy :type Controller%20Object%20Reference) (DamageCauser :type Actor%20Object%20Reference) (HitLocation :type Vector) (FHitComponent :type Primitive%20Component%20Object%20Reference) (BoneName :type Name) (Momentum :type Vector))
  (bind _advancestagetimer (Variables|VaultEncounterManager|GetAdvanceStageTimer))
  (bind _vaultmanager (Variables|VaultEncounterManager|GetVaultManager)
    (:then
      (Class|BGAVividRazorVaultManager|DisplayBossonMinimap _vaultmanager false))
    (:else
      _))
  (if (and (Utilities|Time|IsTimerPausedbyHandle _advancestagetimer) (> (ScalableFloat|GetValueatLevel (Variables|VaultEncounterManager|GetBossPausesSpawning) (Math|Conversions|ToFloat(Integer) (Variables|VaultEncounterManager|GetTeamSize))) 0))
    (Utilities|Time|UnpauseTimerbyHandle _advancestagetimer)
    (else
      _))
  (bind _vaultmanager_2 (Variables|VaultEncounterManager|GetVaultManager)
    (:then
      (bind _vaultactor (Variables|/VividRazorVaultGameplay/Gameplay/Environmental/Manager/BGA_VividRazorVault_Manager.BGA_VividRazorVault_Manager_C|GetVaultActor _vaultmanager_2)
        (:then
          (Class|BGAVividRazorVaultDoorParent|HandleBossDefeated _vaultactor))
        (:else)))
    (:else)))

(fn HandleBossSpawned
  (params (BossPawn :type Fort%20Pawn%20Object%20Reference))
  (bind _advancestagetimer (Variables|VaultEncounterManager|GetAdvanceStageTimer))
  (EventDispatchers|BindEventtoOnDied BossPawn (EventDispatchers|CreateEvent :!event HandleBossDied))
  (Variables|VaultEncounterManager|SetBossSpawned true)
  (Class|FortAthenaLivingWorldPointProviderInterface|DisablePointProvider (Variables|VaultEncounterManager|GetBossPointProvider))
  (if (and (Utilities|Time|IsTimerActivebyHandle _advancestagetimer) (> (ScalableFloat|GetValueatLevel (Variables|VaultEncounterManager|GetBossPausesSpawning) (Math|Conversions|ToFloat(Integer) (Variables|VaultEncounterManager|GetTeamSize))) 0))
    (Utilities|Time|PauseTimerbyHandle _advancestagetimer)))

(fn HandleKillQuotaMet
  (bind _teamsize (Variables|VaultEncounterManager|GetTeamSize))
  (bind _returnvalue (Math|Conversions|ToFloat(Integer) _teamsize))
  (bind _delaybetweenstages (Variables|VaultEncounterManager|GetDelayBetweenStages))
  (bind _returnvalue_2 (ScalableFloat|GetValueatLevel _delaybetweenstages _returnvalue))
  (bind _encounter (Variables|VaultEncounterManager|GetEncounter))
  (Utilities|IsValid _encounter
    (:Is%20Valid
      (Class|FortAthenaLivingWorldEncounterInstance|PauseEncounter _encounter)
      (if (or (== (+ (Variables|/Script/LagerRuntime.FortAthenaLivingWorldEncounterInstance|GetCurrentStageIndex _encounter) 1) (Class|FortAthenaLivingWorldEncounterInstance|GetNumberOfStages _encounter)) (Variables|VaultEncounterManager|GetVaultCracked))
        (Default|EndEncounter)
        (else
          (Development|PrintString (Utilities|String|Append "[Encounter Manager] Kill quota met! The next encounter stage will begin in " (Utilities|String|ToString(Integer64) (Math|Float|RoundtoInteger64 _returnvalue_2)) " seconds.") (Variables|VaultEncounterManager|GetVerboseLogging) true (Variables|VaultEncounterManager|GetDebugOutputColor) 4.0)
          (bind _returnvalue_3 (Utilities|Time|SetTimerbyFunctionName 0 "AdvanceEncounterStage" _returnvalue_2))
          (Variables|VaultEncounterManager|SetAdvanceStageTimer _returnvalue_3)
          (Variables|VaultEncounterManager|SetEncounterStarted true)
          (Variables|VaultEncounterManager|SetStageFlushQueued true))))))

(fn HandlePlayerEnteredEncounterArea
  (params (Player :type Fort%20Pawn%20Object%20Reference))
  (Utilities|Array|AddUnique (Variables|VaultEncounterManager|GetActivePlayers) Player)
  (bind _shouldbehostile (CallFunction|ShouldEnemiesbeHostile))
  (if _shouldbehostile
    (for _array_element (Variables|VaultEncounterManager|GetActiveEnemies)
      (Affiliation|ChangeAttitudeBetweenActors _array_element Player "Hostile"))))

(fn HandlePlayerLeftEncounterArea
  (params (Player :type Fort%20Pawn%20Object%20Reference))
  (Development|PrintString "[Encounter Manager] Player left encounter area!!" (Variables|VaultEncounterManager|GetVerboseLogging) true (Variables|VaultEncounterManager|GetDebugOutputColor))
  (Utilities|Array|RemoveItem (Variables|VaultEncounterManager|GetActivePlayers) Player))

(fn HasKillQuotaBeenMet
  (bind _encounter (Variables|VaultEncounterManager|GetEncounter))
  (Utilities|IsValid _encounter
    (:Is%20Valid
      (return (Math|Boolean|ANDBoolean (not (Variables|VaultEncounterManager|GetStageFlushQueued)) (Variables|VaultEncounterManager|GetEncounterStarted) (or (and (!= (+ (Variables|/Script/LagerRuntime.FortAthenaLivingWorldEncounterInstance|GetCurrentStageIndex _encounter) 1) (Class|FortAthenaLivingWorldEncounterInstance|GetNumberOfStages _encounter)) (>= (/ (Math|Conversions|ToFloat(Integer) (Variables|VaultEncounterManager|GetActorsKilledInCurrentStage)) (Variables|VaultEncounterManager|GetActorsSpawnedInCurrentStage)) (ScalableFloat|GetValueatLevel (Variables|VaultEncounterManager|GetAcceptableKillPercentage) (Math|Conversions|ToFloat(Integer) (Variables|VaultEncounterManager|GetTeamSize))))) (and (== (+ (Variables|/Script/LagerRuntime.FortAthenaLivingWorldEncounterInstance|GetCurrentStageIndex (Variables|VaultEncounterManager|GetEncounter)) 1) (Class|FortAthenaLivingWorldEncounterInstance|GetNumberOfStages (Variables|VaultEncounterManager|GetEncounter))) (== (Utilities|Array|Length (Variables|VaultEncounterManager|GetActiveEnemies)) 0))))))
    (:Is%20Not%20Valid
      (return false))))

(fn InitializeReadyEnemies
  (bind _uninitializedenemies (Variables|VaultEncounterManager|GetUninitializedEnemies))
  (bind _keys (Utilities|Map|Keys (Variables|VaultEncounterManager|GetUninitializedEnemies)))
  (bind (_arrayindex _arrayelement) (Utilities|Array|ReverseforEachLoop _keys)
    (:LoopBody
      (if (Utilities|IsValid _arrayelement :!class /Script/Engine.KismetSystemLibrary)
        (bind _output_get (LocalVariables|SetCurrentEnemy _arrayelement))
        (bind _asfort_athena_aibot_controller (Utilities|Casting|CastToFortAthenaAIBotController (Pawn|GetController _output_get)))
        (LocalVariables|SetCurrentEnemyController _asfort_athena_aibot_controller)
        (bind _values (Utilities|Map|Values (Variables|VaultEncounterManager|GetUninitializedEnemies)))
        (if (>= (- (Utilities|Time|GetTimeSeconds) (Utilities|Array|Get(acopy) _values _arrayindex)) 0.5)
          (Utilities|FlowControl|Sequence
            (:then_0
              (if (and (Variables|VaultEncounterManager|GetCombatStarted) (> (Utilities|Array|Length (Variables|VaultEncounterManager|GetActivePlayers)) 0))
                (Class|FortAthenaAIBotController|AddTargetInfo (LocalVariables|GetCurrentEnemyController) (Utilities|Array|RandomArrayItem (Variables|VaultEncounterManager|GetActivePlayers)) false "Threat_LKP" (ScalableFloat|GetValueatLevel (Variables|VaultEncounterManager|GetTargetForgetTime) (Math|Conversions|ToFloat(Integer) (Variables|VaultEncounterManager|GetTeamSize))) (ScalableFloat|GetValueatLevel (Variables|VaultEncounterManager|GetTargetForgetDistance) (Math|Conversions|ToFloat(Integer) (Variables|VaultEncounterManager|GetTeamSize))))))
            (:then_1
              (if (Variables|VaultEncounterManager|GetEncounterStarted)
                (Class|FortAthenaAIBotController|QueueStim (LocalVariables|GetCurrentEnemyController) (select (Variables|VaultEncounterManager|GetVaultCracked) (Variables|VaultEncounterManager|GetVaultInteriorTargetPoint) (Variables|VaultEncounterManager|GetVaultTargetPoint)) (Math|Vector|SelectVector (Transformation|GetActorLocation (Variables|VaultEncounterManager|GetVaultInteriorTargetPoint)) (Transformation|GetActorLocation (Variables|VaultEncounterManager|GetVaultTargetPoint)) (Variables|VaultEncounterManager|GetVaultCracked)) (- (Math|Vector|SelectVector (Transformation|GetActorLocation (Variables|VaultEncounterManager|GetVaultInteriorTargetPoint)) (Transformation|GetActorLocation (Variables|VaultEncounterManager|GetVaultTargetPoint)) (Variables|VaultEncounterManager|GetVaultCracked)) (Transformation|GetActorLocation (LocalVariables|GetCurrentEnemy))) "Enemy" 1e+02 "None")))
            (:then_2
              (bind _shouldbehostile (CallFunction|ShouldEnemiesbeHostile))
              (if _shouldbehostile
                (for _array_element (Variables|VaultEncounterManager|GetActivePlayers)
                  (Affiliation|ChangeAttitudeBetweenActors (LocalVariables|GetCurrentEnemyController) _array_element "Hostile"))))
            (:then_3
              (Utilities|Map|Remove _uninitializedenemies (LocalVariables|GetCurrentEnemy))
              (Development|PrintString (Utilities|String|Append "[ENCOUNTER MANAGER] Initialized enemy. Enemies left to initialize: " (Utilities|String|ToString(Integer) (Utilities|Map|Length _uninitializedenemies))) (Variables|VaultEncounterManager|GetVerboseLogging) (Variables|VaultEncounterManager|GetVerboseLogging) (Variables|VaultEncounterManager|GetDebugOutputColor) 4.0))))
        (else
          (Utilities|Map|Remove (Variables|VaultEncounterManager|GetUninitializedEnemies) _arrayelement))))
    (:Completed
      (if (and (> (Utilities|Map|Length (Variables|VaultEncounterManager|GetUninitializedEnemies)) 0) (or (not (Utilities|Time|IsTimerActivebyHandle (Variables|VaultEncounterManager|GetEnemyInitializationTimer))) (<= (Utilities|Time|GetTimerRemainingTimebyHandle (Variables|VaultEncounterManager|GetEnemyInitializationTimer)) 0)))
        (Utilities|Time|ClearandInvalidateTimerbyHandle (Variables|VaultEncounterManager|GetEnemyInitializationTimer))
        (bind _returnvalue (Utilities|Time|SetTimerbyFunctionName 0 "InitializeReadyEnemies" 0.2))
        (Variables|VaultEncounterManager|SetEnemyInitializationTimer _returnvalue)))))

(fn IsPawnBoss
  (params (Pawn :type Fort%20Pawn%20Object%20Reference))
  (Utilities|Casting|CastToBP_BotController_Boss_Base (Pawn|GetController Pawn)
    (:then
      (return true))
    (:CastFailed
      (return false))))

(fn IsPawnPlayer
  (params (Pawn :type Pawn%20Object%20Reference))
  (return (Pawn|IsPlayerController (Pawn|GetController Pawn))))

(fn MakeEnemiesHostileTowardsPlayer
  (params (Player :type Fort%20Pawn%20Object%20Reference))
  (for _array_element (Variables|VaultEncounterManager|GetActiveEnemies)
    (Affiliation|ChangeAttitudeBetweenActors _array_element Player "Hostile")))

(fn MakeEnemiesNeutralTowardsPlayer
  (params (Player :type Fort%20Pawn%20Object%20Reference))
  (if (Math|Boolean|NORBoolean (Variables|VaultEncounterManager|GetVaultCracked) (Variables|VaultEncounterManager|GetCombatStarted))
    (for _array_element (Variables|VaultEncounterManager|GetActiveEnemies)
      (bind _asfort_athena_aibot_controller (Utilities|Casting|CastToFortAthenaAIBotController (AI|GetAIController _array_element)))
      (if (!= (Variables|/Script/FortniteAI.FortAthenaAIBotController|GetCurrentAlertLevel _asfort_athena_aibot_controller) (Math|Byte|MakeLiteralByte (Utilities|Enum|LiteralenumEAlertLevel "Threatened")))
        (Affiliation|ChangeAttitudeBetweenActors _array_element Player "Neutral")))))

(fn PauseEncounter
  (bind _encounter (Variables|VaultEncounterManager|GetEncounter))
  (Utilities|IsValid _encounter
    (:Is%20Valid
      (Class|FortAthenaLivingWorldEncounterInstance|PauseEncounter _encounter)
      (Development|PrintString "[Encounter Manager]  Pausing encounter!" (Variables|VaultEncounterManager|GetVerboseLogging) true (Variables|VaultEncounterManager|GetDebugOutputColor) 4.0))))

(fn ReserveAIDensityReservoir
  (params (Reserve :type Boolean))
  (bind _returnvalue (Class|FortAthenaLivingWorldManager|GetLivingWorldManager))
  (bind _self self)
  (Networking|SwitchHasAuthority
    (:Authority
      (if Reserve
        (bind _returnvalue_2 (Class|FortAthenaLivingWorldManager|RegisterDensityReservoir _returnvalue _self (Variables|VaultEncounterManager|GetDensityReservationCount) (Variables|VaultEncounterManager|GetDensityReservationRange)))
        (Fortnite|FortShippingLog _self "Registered density reservoir")
        (Development|PrintString (Utilities|String|ToString(Text) (Utilities|Text|FormatText "NSLOCTEXT(\"[85F5A7F0ABBB11896F921927C316B28B]\", \"E818C5D84249A7EAC321CFA9613B5E4D\", \"[Encounter Manager] Attempted registering density reservoir. Succeeded: {a} \")" _returnvalue_2)) (Variables|VaultEncounterManager|GetVerboseLogging) (Variables|VaultEncounterManager|GetVerboseLogging) (Variables|VaultEncounterManager|GetDebugOutputColor) 6.0)
        (else
          (Class|FortAthenaLivingWorldManager|UnregisterDensityReservoir _returnvalue _self)
          (Fortnite|FortShippingLog _self "Unregistered self from density reservoir"))))))

(fn ResetStageData
  (Variables|VaultEncounterManager|SetActorsSpawnedInCurrentStage 0)
  (Variables|VaultEncounterManager|SetActorsKilledInCurrentStage 0)
  (Variables|VaultEncounterManager|SetStageFlushQueued false))

(fn ShouldEnemiesBeHostile
  (return (Math|Boolean|ORBoolean (and (Utilities|IsValid (Variables|VaultEncounterManager|GetEncounter) :!class /Script/Engine.KismetSystemLibrary) (> (Variables|/Script/LagerRuntime.FortAthenaLivingWorldEncounterInstance|GetCurrentStageIndex (Variables|VaultEncounterManager|GetEncounter)) 0)) (Variables|VaultEncounterManager|GetVaultCracked) (Variables|VaultEncounterManager|GetCombatStarted))))

(fn SpawnBoss
  (Utilities|Time|ClearandInvalidateTimerbyHandle (Variables|VaultEncounterManager|GetBossSpawnTimerHandle))
  (if (not (Variables|VaultEncounterManager|GetBossSpawned))
    (Class|FortAthenaLivingWorldPointProviderInterface|EnablePointProvider_DEPRECATED (Variables|VaultEncounterManager|GetBossPointProvider))
    (else
      _))
  (bind _vaultmanager (Variables|VaultEncounterManager|GetVaultManager)
    (:then
      (bind _vaultactor (Variables|/VividRazorVaultGameplay/Gameplay/Environmental/Manager/BGA_VividRazorVault_Manager.BGA_VividRazorVault_Manager_C|GetVaultActor _vaultmanager)
        (:then
          (Class|BGAVividRazorVaultDoorParent|HandleBossSpawned _vaultactor))
        (:else)))
    (:else)))

(fn SpawnPatrollingGuards
  (bind _teamsize (Variables|VaultEncounterManager|GetTeamSize))
  (bind _patrolencounterdatalist (Variables|VaultEncounterManager|GetPatrolEncounterDataList))
  (bind (_value _returnvalue) (Utilities|Map|Find _patrolencounterdatalist _teamsize))
  (if _returnvalue
    (bind (_isvalid _returnvalue_2) (Default|StartActivityEncounterPatrolPathPointProvider (Game|GetGameState) self _value (Variables|VaultEncounterManager|GetPatrolPointProviders)))
    (EventDispatchers|BindEventtoOnActorSpawned _returnvalue_2 (EventDispatchers|CreateEvent :!event HandleActorSpawned))
    (EventDispatchers|BindEventtoOnActorDied _returnvalue_2 (EventDispatchers|CreateEvent :!event HandleActorDied))
    (else
      (Fortnite|FortShippingLog self (Utilities|String|ToString(Text) (Utilities|Text|FormatText "NSLOCTEXT(\"[85F5A7F0ABBB11896F921927C316B28B]\", \"12705F504F43414FEEB64782156DC322\", \"[Encounter Manager] Attempting to start  a Living World  Encounter from {a} for the specified playlist size {b}, but could not find one in the entity\\'s data!\")" self (Variables|VaultEncounterManager|GetTeamSize))) true))))

(fn StartEncounter
  (bind _teamsize (Variables|VaultEncounterManager|GetTeamSize))
  (bind _returnvalue (Math|Conversions|ToFloat(Integer) _teamsize))
  (bind _delaybetweenstages (Variables|VaultEncounterManager|GetDelayBetweenStages))
  (bind _returnvalue_2 (ScalableFloat|GetValueatLevel _delaybetweenstages _returnvalue))
  (bind _returnvalue_3 (Math|Float|Clamp(Float) _returnvalue_2 0.1 9999.0))
  (bind _teamsize_2 (Variables|VaultEncounterManager|GetTeamSize))
  (bind _bossshouldspawnhere (Variables|VaultEncounterManager|GetBossShouldSpawnHere))
  (bind _bossencounterdatalist (Variables|VaultEncounterManager|GetBossEncounterDataList))
  (bind _regularencounterdatalist (Variables|VaultEncounterManager|GetRegularEncounterDataList))
  (bind _returnvalue_4 (select _bossshouldspawnhere _bossencounterdatalist _regularencounterdatalist))
  (bind (_value _returnvalue_5) (Utilities|Map|Find _returnvalue_4 _teamsize_2))
  (bind _encounter (Variables|VaultEncounterManager|GetEncounter))
  (if _returnvalue_5
    (bind (_isvalid _returnvalue_6) (Default|StartActivityEncounterStaticPointProvider (Game|GetGameState) self _value (Variables|VaultEncounterManager|GetReinforcementPointProviders)))
    (Variables|VaultEncounterManager|SetEncounter _returnvalue_6)
    (CallFunction|ResetStageData)
    (Development|PrintString "[Encounter Manager] Starting stage 1 of the encounter!" (Variables|VaultEncounterManager|GetVerboseLogging) true (Variables|VaultEncounterManager|GetDebugOutputColor) 4.0)
    (Utilities|IsValid (Variables|VaultEncounterManager|GetEncounter)
      (:Is%20Valid
        (EventDispatchers|BindEventtoOnActorDied _encounter (EventDispatchers|CreateEvent :!event HandleActorDied))
        (EventDispatchers|BindEventtoOnActorSpawned _encounter (EventDispatchers|CreateEvent :!event HandleActorSpawned))
        (Variables|VaultEncounterManager|SetEncounterStarted true)
        (Default|CallOnEventStarted)
        (bind _sounddispatcher (Variables|Default|GetSoundDispatcher)
          (:then
            (Class|SoundEventDispatcherVividRazerVault|EncounterBegin _sounddispatcher))
          (:else
            _))
        (Variables|VaultEncounterManager|SetCombatStarted true)
        (if (> (ScalableFloat|GetValueatLevel (Variables|VaultEncounterManager|GetUseTimedStageTriggers) (Math|Conversions|ToFloat(Integer) (Variables|VaultEncounterManager|GetTeamSize))) 0)
          (bind _returnvalue_7 (Utilities|Time|SetTimerbyFunctionName 0 "AdvanceEncounterStage" _returnvalue_3))
          (Variables|VaultEncounterManager|SetAdvanceStageTimer _returnvalue_7)
          (Development|PrintString (Utilities|String|Append "[Encounter Manager] Queueing the next stage in " (Utilities|String|ToString(Float) _returnvalue_3) " seconds!") (Variables|VaultEncounterManager|GetVerboseLogging) true (Variables|VaultEncounterManager|GetDebugOutputColor) 4.0))))
    (else
      (Fortnite|FortShippingLog self (Utilities|String|ToString(Text) (Utilities|Text|FormatText "NSLOCTEXT(\"[85F5A7F0ABBB11896F921927C316B28B]\", \"12705F504F43414FEEB64782156DC322\", \"[Encounter Manager] Attempting to start  a Living World  Encounter from {a} for the specified playlist size {b}, but could not find one in the entity\\'s data!\")" self (Variables|VaultEncounterManager|GetTeamSize))) true))))

(fn SubmitAnalytics
  (Networking|SwitchHasAuthority
    (:Authority
      (if (not (Replay|GetIsPlayingReplay))
        (Utilities|FlowControl|Sequence
          (:then_0
            (bind _returnvalue (Analytics|ConstructAnalyticsAttribute "VaultLocation" (Utilities|String|ToString(Vector) (Transformation|GetActorLocation))))
            (Utilities|Array|Add (LocalVariables|GetAnalyticsEventAttributes) _returnvalue)
            (bind _returnvalue_2 (Analytics|ConstructAnalyticsAttribute :AttributeName "DamageToGuards" :IntValue -1 :floatValue (Variables|Analytics|GetDamageToGuards) :BoolValue false))
            (Utilities|Array|Add (LocalVariables|GetAnalyticsEventAttributes) _returnvalue_2)
            (bind _returnvalue_3 (Analytics|ConstructAnalyticsAttribute :AttributeName "DamageByGuards" :IntValue -1 :floatValue (Variables|Analytics|GetDamageByGuards) :BoolValue false))
            (Utilities|Array|Add (LocalVariables|GetAnalyticsEventAttributes) _returnvalue_3)
            (bind _returnvalue_4 (Analytics|ConstructAnalyticsAttribute :AttributeName "GuardsEliminated" :IntValue (Variables|Analytics|GetActorsKilledTotal) :floatValue -1.0 :BoolValue false))
            (Utilities|Array|Add (LocalVariables|GetAnalyticsEventAttributes) _returnvalue_4))
          (:then_1
            (Analytics|FireAnalyticsEventBlueprintArray "Core.VividRazor.Vault.AI" (LocalVariables|GetAnalyticsEventAttributes))))))))

(fn ToggleReinforcementPointProviders
  (bind _reinforcementpointproviders (Variables|VaultEncounterManager|GetReinforcementPointProviders))
  ; Test to try to see if we disable and reenable the providers, if the AI can start using them. Seems that they get "locked" somewhat if trying to spawn the same AI on the same providers. To Validate with Engineering
  (for _array_element _reinforcementpointproviders
    (Class|FortAthenaLivingWorldPointProviderInterface|DisablePointProvider _array_element))
  (for _array_element_2 _reinforcementpointproviders
    (Class|FortAthenaLivingWorldPointProviderInterface|EnablePointProvider_DEPRECATED _array_element_2)))

(fn UpdateVaultProgress
  (bind _output_get (Variables|VaultEncounterManager|SetCurrentVaultProgress (Math|Float|Clamp(Float) (+ (Variables|VaultEncounterManager|GetCurrentVaultProgress) (/ 1.0 (Variables|VaultEncounterManager|GetNumberOfWeakpoints))))))
  (if (and (not (Variables|VaultEncounterManager|GetBossSpawned)) (>= _output_get (ScalableFloat|GetValueatLevel (Variables|VaultEncounterManager|GetBossSpawnFallbackProgressPercentage) (Math|Conversions|ToFloat(Integer) (Variables|VaultEncounterManager|GetTeamSize)))))
    (CallFunction|SpawnBoss)))

(fn ConstructionScript)

(fn ValidateTeamSize
  (Variables|VaultEncounterManager|SetTeamSize (Math|Integer|Min(Integer) (Math|Integer|Max(Integer) (Variables|VaultEncounterManager|GetTeamSize)) 4))
  (return))
